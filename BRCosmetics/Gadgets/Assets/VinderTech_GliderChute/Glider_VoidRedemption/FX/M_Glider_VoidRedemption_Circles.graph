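# Unlit translucent, two-sided Niagara particle material for the circle effects on the Void Redemption glider.
# A dynamic parameter slides a soft procedural radial disc diagonally across the UVs, and that disc is subtracted from a circle texture to carve the mask.
# The squared mask blends ShadowColor toward particle color through the shared emissive adjustments function.
# The mask is also multiplied by particle alpha.
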Material /BRCosmetics/Gadgets/Assets/VinderTech_GliderChute/Glider_VoidRedemption/FX/M_Glider_VoidRedemption_Circles
graph {
 properties {
  bUsedWithNiagaraMeshParticles=true
  bUsedWithNiagaraSprites=true
  blendMode=BLEND_Translucent
  pixelDepthOffsetMode=PDOM_Legacy
  shadingModel=MSM_Unlit
  twoSided=true
  usedShadingModels=Unlit
 }
 nodes [
  n1 = ParticleColor()
  n2 = TextureSample() {
   samplerType=SAMPLERTYPE_LinearGrayscale
   texture=&/T_Glider_VoidRedemption_Circle
  }
  n3 = DynamicParameter() {defaultValue=<rgba 0 -0.2 1 1>}
  n4 = Distance(A=Add(A=TextureCoordinate(), B=AppendVector(A=@n3.Param1, B=@n3.Param1)) {constB=-0.1}, B=Constant() {r=0.5})
  n5 = Multiply(A=Subtract(A=OneMinus(None=@n4)) {constB=0.6}) {constB=5}
  n6 = Saturate(None=Subtract(A=@n2.R, B=Saturate(None=@n5)))
  MaterialFunctionCall(Emissive=LinearInterpolate(A=VectorParameter() {parameterName=ShadowColor, defaultValue=<rgba 0.0261 0.0639 1 1>}.RGB, B=@n1.RGB, Alpha=Multiply(A=@n6, B=@n6))) {materialFunction=&MF_FX_EmissiveAdjustments}
  Multiply(A=@n1.A, B=@n6)
 ]
}
